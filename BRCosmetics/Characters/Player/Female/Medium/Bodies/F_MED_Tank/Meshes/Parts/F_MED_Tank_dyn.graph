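# Physics asset for the female medium Tank cosmetic body, defining secondary-motion physics for its left and right hat pieces and a tag.
# Capsules cover upper spine, head and tag bones, while thin tilted boxes form the two hat pieces.
# Constraints hang the hat pieces off the head and the tag off spine_05 through a master bone.
# Hat and tag joints have tight twist limits, while the master and spine joints rotate freely.
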PhysicsAsset /BRCosmetics/Characters/Player/Female/Medium/Bodies/F_MED_Tank/Meshes/Parts/F_MED_Tank_dyn
bodies {
 dyn_hat_l [{shapeType=Box, extentX=7.5, extentY=11, extentZ=1, center=<6 0 -1>, rotation=<rot -10 0 10>}]
 dyn_hat_r [{shapeType=Box, extentX=7.5, extentY=11, extentZ=1, center=<6 0 -1>, rotation=<rot -10 0 -10>}]
 dyn_tag [{shapeType=Capsule, length=1.7602, radius=1, center=<3 0 0>, rotation=<rot 90 0 0>}]
 dyn_tag_MSTR [{shapeType=Capsule, length=1, radius=1}]
 head [{shapeType=Capsule, length=6.25, radius=6.75, center=<2.5719 0 0>, rotation=<rot 90 0 0>}]
 spine_04 [{shapeType=Capsule, length=12.9528, radius=7.6365, center=<7.4534 0.0819 0>}]
 spine_05 [{shapeType=Capsule, length=12.9078, radius=7.6816, center=<2.3354 -1.1673 0>}]
}
constraints [
 {bone1Name=head, bone2Name=spine_05, swing1LimitDegrees=45, swing2LimitDegrees=45, twistLimitDegrees=45}
 {bone1Name=dyn_hat_l, bone2Name=head, swing1LimitDegrees=30, swing2LimitDegrees=30, twistLimitDegrees=5}
 {bone1Name=dyn_hat_r, bone2Name=head, swing1LimitDegrees=30, swing2LimitDegrees=30, twistLimitDegrees=5}
 {
  bone1Name=dyn_tag_MSTR
  bone2Name=spine_05
  swing1LimitDegrees=45
  swing1Motion=Free
  swing2LimitDegrees=45
  swing2Motion=Free
  twistLimitDegrees=45
  twistMotion=Free
 }
 {bone1Name=dyn_tag, bone2Name=dyn_tag_MSTR, swing1LimitDegrees=30, swing2LimitDegrees=15, twistLimitDegrees=5}
 {
  bone1Name=spine_05
  bone2Name=spine_04
  swing1LimitDegrees=45
  swing1Motion=Free
  swing2LimitDegrees=45
  swing2Motion=Free
  twistLimitDegrees=45
  twistMotion=Free
 }
]
